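# Secondary-motion AnimBlueprint for the F_MED_Binary female body cosmetic, following the attached parent mesh via Copy Pose From Mesh.
# A serial chain uses weighted Constraint nodes to anchor hose and arm-zipper master bones to spine, pelvis, clavicle and upper-arm bones.
# Root-relative Anim Dynamics adds sprung, limited jiggle to hoses, zippers, pouches and a belt utility item.
# Deform and right-shoulder Control Rigs finish the pose; the EventGraph has no logic.
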AnimBlueprint /BRCosmetics/Characters/Player/Female/Medium/Bodies/F_MED_Binary/Meshes/F_MED_Binary_AnimBP
target_skeleton=&/F_MED_Binary_Skeleton
graphs [
 AnimGraph {nodes [
   n1 = "Constraint - mstr_hose_frnt (PPPP)"!(
    ComponentPose=Local To Component(LocalPose=Copy Pose From Mesh() {bUseAttachedParent=true})
    Alpha=1
    ConstraintWeights_0=0.668
    ConstraintWeights_1:float=0.1933
    ConstraintWeights_2:float=0.098
    ConstraintWeights_3:float=0.0406
   ) -> (Pose) {
    alphaBoolBlend {blendOption=Linear}
    boneToModify {boneName=mstr_hose_frnt}
    constraintSetup [{targetBone {boneName=spine_04}}, {targetBone {boneName=spine_03}}, {targetBone {boneName=spine_02}}, {targetBone {boneName=spine_01}}]
    constraintWeights [1,1,1,1]
   }
   n2 = Anim Dynamics - Bone: dyn_hose_frnt(ComponentPose=@n1, Alpha=1, ExternalForce="0,0,0") {
    angularDampingOverride=0.97
    angularSpringConstant=20000
    bLinearSpring=true
    bOverrideLinearDamping=true
    bUseGravityOverride=true
    gravityScale=0.35
    linearDampingOverride=0.99
    linearSpringConstant=100
    simulationSpace=RootRelative
    alphaBoolBlend {blendOption=Linear}
    boundBone {boneName=dyn_hose_frnt}
    componentAppliedLinearAccClamp=<3000 3000 3000>
    componentLinearAccScale=<0.1 0.1 0.1>
    componentLinearVelScale=<0.9 0.9 0.9>
    gravityOverride {z=-350}
    physicsBodyDefinitions [{
      boundBone {boneName=dyn_hose_frnt}
      boxExtents=<50 50 50>
      constraintSetup {angularLimitsMax=<0 0 0>, angularLimitsMin=<0 0 0>, angularTarget=<0 0 0>, linearAxesMax=<0.75 0.75 0.75>, linearAxesMin=<-0.25 -0.25 -0.25>}
      localJointOffset=<0 0 0>
     }]
    relativeSpaceBone {boneName=pelvis}
   }
   n3 = "Constraint - mstr_hose_side (PPPPP)"!(
    ComponentPose=@n2
    Alpha=1
    ConstraintWeights_0=0.411
    ConstraintWeights_1:float=0.22
    ConstraintWeights_2:float=0.197
    ConstraintWeights_3:float=0.12
    ConstraintWeights_4:float=0.052
   ) -> (Pose) {
    alphaBoolBlend {blendOption=Linear}
    boneToModify {boneName=mstr_hose_side}
    constraintSetup [
     {targetBone {boneName=spine_02}}
     {targetBone {boneName=spine_01}}
     {targetBone {boneName=spine_04}}
     {targetBone {boneName=spine_03}}
     {targetBone {boneName=pelvis}}
    ]
    constraintWeights [1,1,1,1,1]
   }
   n4 = Anim Dynamics - Bone: dyn_hose_side(ComponentPose=@n3, Alpha=1, ExternalForce="0,0,0") {
    angularDampingOverride=0.999
    angularSpringConstant=200
    bAngularSpring=true
    bLinearSpring=true
    bOverrideAngularDamping=true
    bOverrideLinearDamping=true
    bUseGravityOverride=true
    gravityScale=0.35
    linearDampingOverride=0.9995
    linearSpringConstant=100
    simulationSpace=RootRelative
    alphaBoolBlend {blendOption=Linear}
    boundBone {boneName=dyn_hose_side}
    componentAppliedLinearAccClamp=<5000 5000 5000>
    componentLinearAccScale=<0.25 0.25 0.25>
    componentLinearVelScale=<0.9 0.9 0.9>
    gravityOverride {z=-375}
    physicsBodyDefinitions [{
      boundBone {boneName=dyn_hose_side}
      boxExtents=<50 50 50>
      constraintSetup {
       angularLimitsMax=<0 2.5 2.5>
       angularLimitsMin=<0 -2.5 -2.5>
       angularTarget=<1 0 0>
       linearAxesMax=<1.8 1.25 3>
       linearAxesMin=<-0.4 -1.25 -0.5>
      }
      localJointOffset=<0 0 0>
     }]
    relativeSpaceBone {boneName=pelvis}
   }
   n5 = "Constraint - mstr_hose_back (PPPPPP)"!(
    ComponentPose=@n4
    Alpha=1
    ConstraintWeights_0=0.3151
    ConstraintWeights_1:float=0.31
    ConstraintWeights_2:float=0.2335
    ConstraintWeights_3:float=0.1373
    ConstraintWeights_4:float=0.0024
    ConstraintWeights_5:float=0.0017
   ) -> (Pose) {
    alphaBoolBlend {blendOption=Linear}
    boneToModify {boneName=mstr_hose_back}
    constraintSetup [
     {targetBone {boneName=spine_02}}
     {targetBone {boneName=spine_03}}
     {targetBone {boneName=spine_04}}
     {targetBone {boneName=spine_01}}
     {targetBone {boneName=clavicle_l}}
     {targetBone {boneName=spine_05}}
    ]
    constraintWeights [1,1,1,1,1,1]
   }
   n6 = Anim Dynamics - Bone: dyn_hose_back(ComponentPose=@n5, Alpha=1, ExternalForce="0,0,0") {
    angularDampingOverride=0.97
    angularSpringConstant=20000
    bLinearSpring=true
    bOverrideLinearDamping=true
    bUseGravityOverride=true
    gravityScale=0.35
    linearDampingOverride=0.99
    linearSpringConstant=100
    simulationSpace=RootRelative
    alphaBoolBlend {blendOption=Linear}
    boundBone {boneName=dyn_hose_back}
    componentAppliedLinearAccClamp=<3000 3000 3000>
    componentLinearAccScale=<0.1 0.1 0.1>
    componentLinearVelScale=<0.9 0.9 0.9>
    gravityOverride {z=-350}
    physicsBodyDefinitions [{
      boundBone {boneName=dyn_hose_back}
      boxExtents=<50 50 50>
      constraintSetup {angularLimitsMax=<0 0 0>, angularLimitsMin=<0 0 0>, angularTarget=<0 0 0>, linearAxesMax=<1.25 0.25 2.5>, linearAxesMin=<-0.5 -2 -1>}
      localJointOffset=<0 0 0>
     }]
    relativeSpaceBone {boneName=pelvis}
   }
   n7 = "Constraint - mstr_arm_zipper_l (PPP)"!(ComponentPose=@n6, Alpha=1, ConstraintWeights_0=0.168, ConstraintWeights_1:float=0.65, ConstraintWeights_2:float=0.182) -> (Pose) {
    alphaBoolBlend {blendOption=Linear}
    boneToModify {boneName=mstr_arm_zipper_l}
    constraintSetup [{targetBone {boneName=upperarm_l}}, {targetBone {boneName=upperarm_twist_01_l}}, {targetBone {boneName=upperarm_twist_02_l}}]
    constraintWeights [1,0,1]
   }
   n8 = Anim Dynamics - Bone: arm_zipper_01_l(ComponentPose=@n7, Alpha=1, ExternalForce="0,0,0") {
    angularDampingOverride=0.98
    angularSpringConstant=500
    bAngularSpring=true
    bOverrideAngularDamping=true
    bUseGravityOverride=true
    gravityScale=0.35
    simulationSpace=RootRelative
    alphaBoolBlend {blendOption=Linear}
    boundBone {boneName=arm_zipper_01_l}
    componentLinearAccScale=<1 1 1>
    componentLinearVelScale=<1 1 1>
    gravityOverride {z=-980}
    physicsBodyDefinitions [{
      boundBone {boneName=arm_zipper_01_l}
      boxExtents=<4 4 4>
      constraintSetup {angularLimitsMax=<0 10 0>, angularLimitsMin=<0 -10 -30>, angularTarget=<1 0 0>}
      localJointOffset=<1.8 0 0>
     }]
   }
   n9 = "Constraint - mstr_arm_zipper_r (PPP)"!(ComponentPose=@n8, Alpha=1, ConstraintWeights_0=0.168, ConstraintWeights_1:float=0.65, ConstraintWeights_2:float=0.182) -> (Pose) {
    alphaBoolBlend {blendOption=Linear}
    boneToModify {boneName=mstr_arm_zipper_r}
    constraintSetup [{targetBone {boneName=upperarm_r}}, {targetBone {boneName=upperarm_twist_01_r}}, {targetBone {boneName=upperarm_twist_02_r}}]
    constraintWeights [1,0,1]
   }
   n10 = Anim Dynamics - Bone: arm_zipper_01_r(ComponentPose=@n9, Alpha=1, ExternalForce="0,0,0") {
    angularDampingOverride=0.98
    angularSpringConstant=500
    bAngularSpring=true
    bOverrideAngularDamping=true
    bUseGravityOverride=true
    gravityScale=0.35
    simulationSpace=RootRelative
    alphaBoolBlend {blendOption=Linear}
    boundBone {boneName=arm_zipper_01_r}
    componentLinearAccScale=<1 1 1>
    componentLinearVelScale=<1 1 1>
    gravityOverride {z=-980}
    physicsBodyDefinitions [{
      boundBone {boneName=arm_zipper_01_r}
      boxExtents=<4 4 4>
      constraintSetup {angularLimitsMax=<0 15 0>, angularLimitsMin=<0 -15 -30>, angularTarget=<1 0 0>}
      localJointOffset=<-1.8 0 0>
     }]
   }
   n11 = Anim Dynamics - Bone: dyn_pouch_01_l(ComponentPose=@n10, Alpha=1, ExternalForce="0,0,0") {
    angularDampingOverride=0.99
    angularSpringConstant=200
    bAngularSpring=true
    bOverrideAngularDamping=true
    bUseGravityOverride=true
    simulationSpace=RootRelative
    alphaBoolBlend {blendOption=Linear}
    boundBone {boneName=dyn_pouch_01_l}
    componentAppliedLinearAccClamp=<2500 2500 2500>
    componentLinearAccScale=<0.1 0.1 0.1>
    componentLinearVelScale=<1 1 1>
    gravityOverride {z=-250}
    physicsBodyDefinitions [{
      boundBone {boneName=dyn_pouch_01_l}
      boxExtents=<12.5 12.5 12.5>
      constraintSetup {angularLimitsMax=<2.5 2.5 5>, angularLimitsMin=<-2.5 -2.5 -5>, angularTarget=<1 0 0>}
      localJointOffset=<1.25 0 0>
     }]
   }
   n12 = Anim Dynamics - Bone: dyn_pouch_01_r(ComponentPose=@n11, Alpha=1, ExternalForce="0,0,0") {
    angularDampingOverride=0.99
    angularSpringConstant=200
    bAngularSpring=true
    bOverrideAngularDamping=true
    bUseGravityOverride=true
    gravityScale=0.35
    simulationSpace=RootRelative
    alphaBoolBlend {blendOption=Linear}
    boundBone {boneName=dyn_pouch_01_r}
    componentAppliedLinearAccClamp=<2500 2500 2500>
    componentLinearAccScale=<0.1 0.1 0.1>
    componentLinearVelScale=<1 1 1>
    gravityOverride {z=-250}
    physicsBodyDefinitions [{
      boundBone {boneName=dyn_pouch_01_r}
      boxExtents=<12.5 12.5 12.5>
      constraintSetup {angularLimitsMax=<2.5 2.5 2.5>, angularLimitsMin=<-2.5 -2.5 -2.5>, angularTarget=<0 0 0>}
      localJointOffset=<1.25 0 0>
     }]
   }
   n13 = Anim Dynamics - Bone: belt_util_01_r(ComponentPose=@n12, Alpha=1, ExternalForce="0,0,0") {
    angularDampingOverride=0.98
    angularSpringConstant=250
    bAngularSpring=true
    bOverrideAngularDamping=true
    bUseGravityOverride=true
    gravityScale=0.35
    simulationSpace=RootRelative
    alphaBoolBlend {blendOption=Linear}
    boundBone {boneName=belt_util_01_r}
    componentAppliedLinearAccClamp=<2750 2750 2750>
    componentLinearAccScale=<0.2 0.2 0.2>
    componentLinearVelScale=<1 1 1>
    gravityOverride {z=-225}
    physicsBodyDefinitions [{
      boundBone {boneName=belt_util_01_r}
      boxExtents=<12.5 12.5 12.5>
      constraintSetup {angularLimitsMax=<2.5 2.5 2.5>, angularLimitsMin=<-2.5 -2.5 -2.5>, angularTarget=<1 0 0>}
      localJointOffset=<1.5 0 0>
     }]
   }
   n14 = Control Rig(Source=Component To Local(ComponentPose=@n13)) {
    alphaBoolBlend {blendOption=Linear}
    controlRigAssetReference {blueprintRigClass=&/F_MED_Binary_DeformRig_ControlRig.F_MED_Binary_DeformRig_ControlRig_C}
   }
   n15 = Control Rig(Source=@n14) {
    alphaBoolBlend {blendOption=Linear}
    controlRigAssetReference {blueprintRigClass=&/F_MED_Binary_rightShoulder_CtrlRig.F_MED_Binary_rightShoulder_CtrlRig_C}
   }
   Output Pose(Result=@n15) {layerGroup=None}
  ]}
 EventGraph {nodes [Event BlueprintUpdateAnimation()]}
]
variables []
