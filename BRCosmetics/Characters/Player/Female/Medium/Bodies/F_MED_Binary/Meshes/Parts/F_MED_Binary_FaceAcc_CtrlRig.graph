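# Control Rig for the F_MED_Binary face accessory that adjusts the two-segment left and right ponytails as the arms and head move.
# Spherical pose readers on both upper arms offset aim targets and bones, and write SPR curves, including a left-arm rifle pose.
# A head reader shifts the right up target through an eased curve, then Aim nodes orient all four ponytail bones.
# The head_ctrl-to-head transform node is not connected to execution.
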ControlRigBlueprint /BRCosmetics/Characters/Player/Female/Medium/Bodies/F_MED_Binary/Meshes/Parts/F_MED_Binary_FaceAcc_CtrlRig
graphs [
 RigVMFunctionLibrary {links [], nodes []}
 RigVMModel {nodes [
   RigUnit_BeginExecution = &RigUnit_BeginExecution()
   n1 = /Script/ControlRig.RigUnit_SequenceExecution(ExecuteContext=@RigUnit_BeginExecution.ExecutePin) {is=struct RigVMUnitNode}
   n2 = &RigUnit_SphericalPoseReader(
    ActiveRegionScaleFactors=(PositiveWidth=1,NegativeWidth=1,PositiveHeight=2,NegativeHeight=0.6)
    ActiveRegionSize=0.1
    Debug=(bDrawDebug=false,bDrawLocalAxes=false,DebugScale=25,DebugSegments=20,DebugThickness=0.25)
    DriverAxis=(X=-1,Y=0,Z=0)
    DriverItem=(Type=Bone,Name="upperarm_r")
    FalloffRegionScaleFactors=(PositiveWidth=0.5,NegativeWidth=1,PositiveHeight=3,NegativeHeight=1)
    FalloffSize=0.3
    OptionalParentItem=(Type=Bone)
    RotationOffset=(X=-110,Y=-20,Z=20)
    ExecutePin=@n1.B
   )
   > /Script/ControlRig.RigUnit_SetTranslation(
    Item=(Type=Bone,Name="ponytail_01_aim_target_r")
    Space=LocalSpace
    bPropagateToChildren=True
    Value=/Script/RigVM.RigVMFunction_MathVectorAdd(
     B=(X=10,Y=0,Z=0)
     A=&RigUnit_GetTransform(Item=(Type=Bone,Name="ponytail_01_aim_target_r"), Space=LocalSpace, bInitial=false).Transform.Translation
    ) {is=struct RigVMUnitNode, template="Add::Execute(in A,in B,out Result)"}.Result
    Weight=@n2.OutputParam
   ) {is=struct RigVMUnitNode, template="Set Transform::Execute(in Item,in Space,in bInitial,in Value,in Weight,in bPropagateToChildren)"}
   > &RigUnit_SetTranslation(
    Item=(Type=Bone,Name="ponytail_02_aim_target_r")
    Space=LocalSpace
    bPropagateToChildren=True
    Value=&RigVMFunction_MathVectorAdd(
     B=(X=6,Y=-2,Z=-1.5)
     A=&RigUnit_GetTransform(Item=(Type=Bone,Name="ponytail_02_aim_target_r"), Space=LocalSpace).Transform.Translation
    ) {template="Add::Execute(in A,in B,out Result)"}.Result
    Weight=@n2.OutputParam
   ) {template="Set Transform::Execute(in Item,in Space,in bInitial,in Value,in Weight,in bPropagateToChildren)"}
   > &RigUnit_SetTranslation(
    Item=(Type=Bone,Name="ponytail_02_up_target_r")
    Space=LocalSpace
    bPropagateToChildren=True
    Value=&RigVMFunction_MathVectorAdd(
     B=(X=0,Y=0,Z=0)
     A=&RigUnit_GetTransform(Item=(Type=Bone,Name="ponytail_02_up_target_r"), Space=LocalSpace).Transform.Translation
    ) {template="Add::Execute(in A,in B,out Result)"}.Result
    Weight=@n2.OutputParam
   ) {template="Set Transform::Execute(in Item,in Space,in bInitial,in Value,in Weight,in bPropagateToChildren)"}
   > /Script/ControlRig.RigUnit_SetCurveValue(Curve=SPR_Upperarm_r, Value=@n2.OutputParam) {is=struct RigVMUnitNode}
   n3 = &RigUnit_SequenceExecution(ExecuteContext=@n1.D)
   n4 = &RigUnit_SphericalPoseReader(
    ActiveRegionScaleFactors=(PositiveWidth=1,NegativeWidth=1,PositiveHeight=1,NegativeHeight=1)
    ActiveRegionSize=0.05
    Debug=(bDrawDebug=false,DebugScale=25,DebugSegments=20,DebugThickness=0.25)
    DriverAxis=(X=0,Y=1,Z=0)
    DriverItem=(Type=Bone,Name="head")
    FalloffRegionScaleFactors=(PositiveWidth=1,NegativeWidth=1,PositiveHeight=1,NegativeHeight=1)
    FalloffSize=0.15
    OptionalParentItem=(Type=Bone,Name="spine_05")
    RotationOffset=(X=120,Y=0,Z=0)
    ExecutePin=@n3.A
   )
   > &RigUnit_ModifyTransforms(
    ItemToModify=((Item=(Type=Bone,Name="ponytail_02_up_target_r"),Transform=(Rotation=(X=0,Y=0,Z=0,W=1),Translation=(X=5,Y=-3,Z=0),Scale3D=(X=1,Y=1,Z=1))))
    Weight=&RigVMFunction_AnimEvalRichCurve(Curve=(EditorCurveData=(Keys=((InterpMode=RCIM_Cubic),(Time=1,Value=1)))), Value=@n4.OutputParam).Result
   )
   /Script/ControlRig.RigUnit_AimItem(
    DebugSettings=(Scale=10,WorldOffset=(Rotation=(X=0,Y=0,Z=0,W=1),Translation=(X=0,Y=0,Z=0),Scale3D=(X=1,Y=1,Z=1)))
    Item=(Type=Bone,Name="ponytail_01_r")
    Primary=(Weight=1,Axis=(X=1,Y=0,Z=0),Target=(X=0,Y=0,Z=0),Kind=Location,Space=(Type=Bone,Name="ponytail_01_aim_target_r"))
    Secondary=(Weight=1,Axis=(X=0,Y=-1,Z=0),Target=(X=0,Y=0,Z=0),Kind=Location,Space=(Type=Bone,Name="ponytail_01_up_target_r"))
    Weight=1
    ExecutePin=@n3.B
   ) {is=struct RigVMUnitNode}
   > &RigUnit_AimItem(
    DebugSettings=(Scale=10,WorldOffset=(Rotation=(X=0,Y=0,Z=0,W=1),Translation=(X=0,Y=0,Z=0),Scale3D=(X=1,Y=1,Z=1)))
    Item=(Type=Bone,Name="ponytail_02_r")
    Primary=(Weight=1,Axis=(X=1,Y=0,Z=0),Target=(X=0,Y=0,Z=0),Kind=Location,Space=(Type=Bone,Name="ponytail_02_aim_target_r"))
    Secondary=(Weight=1,Axis=(X=0,Y=-1,Z=0),Target=(X=0,Y=0,Z=0),Kind=Location,Space=(Type=Bone,Name="ponytail_02_up_target_r"))
    Weight=1
   )
   > &RigUnit_AimItem(
    DebugSettings=(Scale=10,WorldOffset=(Rotation=(X=0,Y=0,Z=0,W=1),Translation=(X=0,Y=0,Z=0),Scale3D=(X=1,Y=1,Z=1)))
    Item=(Type=Bone,Name="ponytail_01_l")
    Primary=(Weight=1,Axis=(X=1,Y=0,Z=0),Target=(X=0,Y=0,Z=0),Kind=Location,Space=(Type=Bone,Name="ponytail_01_aim_target_l"))
    Secondary=(Weight=1,Axis=(X=0,Y=-1,Z=0),Target=(X=0,Y=0,Z=0),Kind=Location,Space=(Type=Bone,Name="ponytail_01_up_target_l"))
    Weight=1
   )
   > &RigUnit_AimItem(
    DebugSettings=(Scale=10,WorldOffset=(Rotation=(X=0,Y=0,Z=0,W=1),Translation=(X=0,Y=0,Z=0),Scale3D=(X=1,Y=1,Z=1)))
    Item=(Type=Bone,Name="ponytail_02_l")
    Primary=(Weight=1,Axis=(X=1,Y=0,Z=0),Target=(X=0,Y=0,Z=0),Kind=Location,Space=(Type=Bone,Name="ponytail_02_aim_target_l"))
    Secondary=(Weight=1,Axis=(X=0,Y=-1,Z=0),Target=(X=0,Y=0,Z=0),Kind=Location,Space=(Type=Bone,Name="ponytail_02_up_target_l"))
    Weight=1
   )
   n5 = &RigUnit_SphericalPoseReader(
    ActiveRegionScaleFactors=(PositiveWidth=1,NegativeWidth=1,PositiveHeight=1.5,NegativeHeight=1.5)
    ActiveRegionSize=0.05
    Debug=(bDrawDebug=false,DebugScale=25,DebugSegments=20,DebugThickness=0.25)
    DriverAxis=(X=1,Y=0,Z=0)
    DriverItem=(Type=Bone,Name="upperarm_l")
    FalloffRegionScaleFactors=(PositiveWidth=0.5,NegativeWidth=0.5,PositiveHeight=1,NegativeHeight=1)
    FalloffSize=0.2
    OptionalParentItem=(Type=Bone)
    RotationOffset=(X=95,Y=160,Z=-30)
    ExecutePin=@n1.C
   )
   > &RigUnit_SetTranslation(
    Item=(Type=Bone,Name="ponytail_01_aim_target_l")
    Space=LocalSpace
    bPropagateToChildren=True
    Value=&RigVMFunction_MathVectorAdd(
     B=(X=4,Y=4,Z=0)
     A=&RigUnit_GetTransform(Item=(Type=Bone,Name="ponytail_01_aim_target_l"), Space=LocalSpace).Transform.Translation
    ) {template="Add::Execute(in A,in B,out Result)"}.Result
    Weight=@n5.OutputParam
   ) {template="Set Transform::Execute(in Item,in Space,in bInitial,in Value,in Weight,in bPropagateToChildren)"}
   > &RigUnit_SetTranslation(
    Item=(Type=Bone,Name="ponytail_02_aim_target_l")
    Space=LocalSpace
    bPropagateToChildren=True
    Value=&RigVMFunction_MathVectorAdd(
     B=(X=-3,Y=0,Z=0)
     A=&RigUnit_GetTransform(Item=(Type=Bone,Name="ponytail_02_aim_target_l"), Space=LocalSpace).Transform.Translation
    ) {template="Add::Execute(in A,in B,out Result)"}.Result
    Weight=@n5.OutputParam
   ) {template="Set Transform::Execute(in Item,in Space,in bInitial,in Value,in Weight,in bPropagateToChildren)"}
   > &RigUnit_SetTranslation(
    Item=(Type=Bone,Name="ponytail_02_up_target_l")
    Space=LocalSpace
    bPropagateToChildren=True
    Value=&RigVMFunction_MathVectorAdd(
     B=(X=5,Y=0,Z=0)
     A=&RigUnit_GetTransform(Item=(Type=Bone,Name="ponytail_02_up_target_l"), Space=LocalSpace).Transform.Translation
    ) {template="Add::Execute(in A,in B,out Result)"}.Result
    Weight=@n5.OutputParam
   ) {template="Set Transform::Execute(in Item,in Space,in bInitial,in Value,in Weight,in bPropagateToChildren)"}
   > n6 = &RigUnit_SphericalPoseReader(
    ActiveRegionScaleFactors=(PositiveWidth=1,NegativeWidth=1,PositiveHeight=1.5,NegativeHeight=1.5)
    ActiveRegionSize=0.12
    Debug=(bDrawDebug=false,DebugScale=25,DebugSegments=20,DebugThickness=0.25)
    DriverAxis=(X=1,Y=0,Z=0)
    DriverItem=(Type=Bone,Name="upperarm_l")
    FalloffRegionScaleFactors=(PositiveWidth=0.5,NegativeWidth=0.5,PositiveHeight=1,NegativeHeight=1)
    FalloffSize=0.25
    OptionalParentItem=(Type=Bone)
    RotationOffset=(X=105,Y=-5,Z=-79)
   )
   > &RigUnit_SetTranslation(
    Item=(Type=Bone,Name="ponytail_01_l")
    Space=LocalSpace
    bPropagateToChildren=True
    Value=&RigVMFunction_MathVectorAdd(
     B=(X=-0.5,Y=0,Z=-0.5)
     A=&RigUnit_GetTransform(Item=(Type=Bone,Name="ponytail_01_l"), Space=LocalSpace).Transform.Translation
    ) {template="Add::Execute(in A,in B,out Result)"}.Result
    Weight=@n6.OutputParam
   ) {template="Set Transform::Execute(in Item,in Space,in bInitial,in Value,in Weight,in bPropagateToChildren)"}
   > &RigUnit_SetTranslation(
    Item=(Type=Bone,Name="ponytail_01_aim_target_l")
    Space=LocalSpace
    bPropagateToChildren=True
    Value=&RigVMFunction_MathVectorAdd(
     B=(X=4,Y=0,Z=0)
     A=&RigUnit_GetTransform(Item=(Type=Bone,Name="ponytail_01_aim_target_l"), Space=LocalSpace).Transform.Translation
    ) {template="Add::Execute(in A,in B,out Result)"}.Result
    Weight=@n6.OutputParam
   ) {template="Set Transform::Execute(in Item,in Space,in bInitial,in Value,in Weight,in bPropagateToChildren)"}
   > &RigUnit_SetCurveValue(Curve=SPR_Upperarm_l, Value=@n5.OutputParam)
   > &RigUnit_SetCurveValue(Curve=SPR_Rifle_Upperarm_l, Value=@n6.OutputParam)
   &RigUnit_ModifyTransforms(
    ExecutePin=()
    ItemToModify=((Item=(Type=Bone,Name="head"),Transform=(Rotation=(X=0,Y=0,Z=0,W=1),Translation=(X=0,Y=0,Z=0),Scale3D=(X=1,Y=1,Z=1))))
    Weight=1
    ItemToModify.0.Transform=&RigUnit_GetTransform(Item=(Type=Control,Name="head_ctrl"), Space=LocalSpace).Transform
   )
  ]}
]
hierarchy [
 root {loc=<0 0 0>, rot=<0 0 0 1>}
 pelvis {parent=root, loc=<0 -0.5301 98.0571>, rot=<0.0127 0.707 -0.0127 -0.707>}
 spine_01 {parent=pelvis, loc=<6.1208 -0.9328 0>, rot=<0 0 0.034 -0.9994>}
 spine_02 {parent=spine_01, loc=<6.5996 -0.2095 0>, rot=<0 0 0.0238 -0.9997>}
 spine_03 {parent=spine_02, loc=<6.4914 -0.1237 0>, rot=<0 0 0.0149 -0.9999>}
 spine_04 {parent=spine_03, loc=<6.9598 0.0794 0>, rot=<0 0 -0.0132 -0.9999>}
 spine_05 {parent=spine_04, loc=<6.9987 0 0>, rot=<0 0 -0.0015 -1>}
 clavicle_l {parent=spine_05, loc=<5.4831 -0.1053 -3.2946>, rot=<0.0465 -0.7289 -0.009 -0.6829>}
 ponytail_01_aim_target_l {parent=clavicle_l, loc=<4.1545 4.6153 3.0564>, rot=<-0.699 0.0185 0.703 0.1301>}
 ponytail_02_aim_target_l {parent=clavicle_l, loc=<6.1494 9.773 -0.625>, rot=<0.7168 0.1233 -0.6309 -0.2703>}
 ponytail_02_up_target_l {parent=clavicle_l, loc=<4.8964 7.6831 12.0155>, rot=<0.7168 0.1233 -0.6309 -0.2703>}
 upperarm_l {parent=clavicle_l, loc=<9.8658 -1.2136 -0.4105>, rot=<0.0164 -0.3245 -0.0269 -0.9454>}
 lowerarm_l {parent=upperarm_l, loc=<26.465 -0.8345 0.259>, rot=<0.0036 0.0161 -0.2298 -0.9731>}
 clavicle_r {parent=spine_05, loc=<5.4831 -0.1053 3.2946>, rot=<0.7289 0.0465 -0.6829 0.009>}
 ponytail_01_aim_target_r {parent=clavicle_r, loc=<-5.134 -8.6044 1.1336>, rot=<-0.0749 0.7238 0.1803 -0.6618>}
 ponytail_02_up_target_r {parent=clavicle_r, loc=<-4.3228 -8.877 -11.3142>, rot=<-0.1101 0.7151 0.1934 -0.6627>}
 ponytail_02_aim_target_r {parent=clavicle_r, loc=<-4.8818 -9.0114 0.0927>, rot=<-0.1101 0.7151 0.1934 -0.6627>}
 upperarm_r {parent=clavicle_r, loc=<-9.8658 1.2136 0.4104>, rot=<0.0164 -0.3245 -0.0269 -0.9454>}
 lowerarm_r {parent=upperarm_r, loc=<-26.4651 0.8345 -0.259>, rot=<0.0036 0.0161 -0.2298 -0.9731>}
 neck_01 {parent=spine_05, loc=<11.3138 0.1529 0>, rot=<0 0 -0.2003 -0.9797>}
 neck_02 {parent=neck_01, loc=<4.5627 0 0>, rot=<0 0 0.0504 -0.9987>}
 head {parent=neck_02, loc=<5.3914 0 0>, rot=<0 0 0.115 -0.9934>}
 ponytail_01_l {parent=head, loc=<1.6558 -1.6154 -7.6084>, rot=<-0.0277 0.0402 0.9872 0.1518>}
 ponytail_02_l {parent=ponytail_01_l, loc=<8.4448 0 0>, rot=<-0.099 -0.0193 0.1196 -0.9877>}
 dyn_ponytail_hair_01_l {parent=ponytail_02_l, loc=<4.5966 -0.6532 -3.2258>, rot=<-0.5048 -0.3247 -0.05 -0.7983>}
 dyn_hair_01_l {parent=head, loc=<4.1995 5.2302 -6.4766>, rot=<0.0864 0.3475 0.9309 0.0723>}
 dyn_bangs_01_r {parent=head, loc=<10.2069 7.9306 2.5326>, rot=<0.0408 -0.1995 0.9637 0.1727>}
 dyn_bangs_01_l {parent=head, loc=<10.5615 8.0859 -1.3122>, rot=<-0.2341 0.1512 0.9522 0.125>}
 dyn_hair_01_r {parent=head, loc=<2.8411 4.0317 8.6314>, rot=<-0.2084 -0.2348 0.9474 0.0625>}
 dyn_hair_02_r {parent=dyn_hair_01_r, loc=<3.7552 0 0>, rot=<-0.2106 -0.0322 -0.1703 -0.9621>}
 ponytail_01_r {parent=head, loc=<0.792 -1.8415 8.208>, rot=<-0.0438 -0.1246 0.9619 0.2394>}
 ponytail_02_r {parent=ponytail_01_r, loc=<8.1146 0 0>, rot=<-0.0457 0.0481 -0.0276 -0.9974>}
 dyn_ponytail_hair_01_r {parent=ponytail_02_r, loc=<2.8831 -1.4245 1.2969>, rot=<0.3075 0.491 0.0631 -0.8127>}
 dyn_ponytail_hair_02_r {parent=dyn_ponytail_hair_01_r, loc=<2.7163 -0.1919 -0.0466>, rot=<0.1371 0.0556 -0.2449 -0.9582>}
 ponytail_01_up_target_l {parent=head, loc=<-0.0661 5.7318 -10.0289>, rot=<0.0337 0.1328 0.9879 0.0721>}
 ponytail_01_up_target_r {parent=head, loc=<0.9794 4.9391 8.6885>, rot=<0.0158 -0.1051 0.9791 0.1735>}
 head_ctrl {controlType=EULER_TRANSFORM, parent=head, type=Control, loc=<0 0 0>, rot=<0 0 0 1>}
 SPR_Upperarm_r {type=Curve, loc=<0 0 0>, rot=<0 0 0 1>}
 SPR_Upperarm_l {type=Curve, loc=<0 0 0>, rot=<0 0 0 1>}
 SPR_Rifle_Upperarm_l {type=Curve, loc=<0 0 0>, rot=<0 0 0 1>}
]
