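# Physics asset for the Binary female medium character's face accessory mesh, defining a two-bone right-side ponytail chain.
# Constraints link ponytail_02_r to ponytail_01_r and that to the head, with limited swing and tight twist; the head body is only a tiny anchor-sized capsule.
# Capsule bodies on the upper spine, right clavicle and right upper arm approximate the torso and shoulder, with no constraints tying them to the chain.
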PhysicsAsset /BRCosmetics/Characters/Player/Female/Medium/Bodies/F_MED_Binary/Meshes/Parts/F_MED_Binary_FaceAcc_Physics
bodies {
 clavicle_r [
  {
   shapeType=Capsule
   length=5.7724
   radius=5.3158
   shapeName=clavicle_l_capsule
   center=<-9.2661 0.0838 -0.1635>
   rotation=<rot -83.6899 0.0023 171.4322>
  }
  {
   shapeType=Capsule
   length=13.5108
   radius=7.2021
   shapeName=clavicle_r_capsule
   center=<-6.5654 7.0245 -5.096>
   rotation=<rot -90 172.4476 -172.4482>
  }
 ]
 head [{shapeType=Capsule, length=0, radius=0.5, shapeName=head_capsule}]
 ponytail_01_r [{shapeType=Capsule, length=5.6499, radius=1.5988, shapeName=ponytail_01_l_capsule, center=<3.9305 0 0>, rotation=<rot -90 0 179.9996>}]
 ponytail_02_r [
  {
   shapeType=Capsule
   length=9.4749
   radius=1.86
   shapeName=ponytail_02_l_capsule
   center=<5.9023 -0.187 0>
   rotation=<rot -90 -161.5699 154.3147>
  }
  {
   shapeType=Capsule
   length=5.3221
   radius=1.5111
   shapeName=ponytail_02_l_capsule
   center=<6.0637 -0.2577 1.747>
   rotation=<rot -90 -161.5699 -25.946>
  }
  {
   shapeType=Capsule
   length=5.3319
   radius=1.5933
   shapeName=ponytail_02_l_capsule
   center=<5.9988 -0.4056 -1.9016>
   rotation=<rot -90 -161.5699 -26.6319>
  }
 ]
 spine_04 [
  {shapeType=Capsule, length=6.045, radius=10.7202, shapeName=spine_04_capsule, center=<0.9746 2.5449 0>}
  {shapeType=Capsule, length=15.2496, radius=4.6781, shapeName=spine_04_capsule, center=<6.707 4.5909 0>}
 ]
 spine_05 [{shapeType=Capsule, length=14.7057, radius=6.675, shapeName=spine_05_capsule, center=<4.4526 -0.8868 0>}]
 upperarm_r [
  {shapeType=Capsule, length=22.1352, radius=3.8731, shapeName=upperarm_r_capsule, center=<-12.3848 0 0>, rotation=<rot -90 90 -90>}
  {shapeType=Capsule, length=22.1352, radius=3.8731, shapeName=upperarm_r_capsule, center=<-12.5287 -2.7297 -5.3913>, rotation=<rot -90 0 0>}
 ]
}
constraints [
 {bone1Name=ponytail_01_r, bone2Name=head, swing1LimitDegrees=30, swing2LimitDegrees=15, twistLimitDegrees=5}
 {bone1Name=ponytail_02_r, bone2Name=ponytail_01_r, swing1LimitDegrees=40, swing2LimitDegrees=25, twistLimitDegrees=5}
]
