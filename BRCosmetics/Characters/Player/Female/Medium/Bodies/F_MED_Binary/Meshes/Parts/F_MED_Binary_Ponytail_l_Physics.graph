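# Physics asset for the left ponytail part of the female medium Binary player character, defining its collision bodies and joint limits.
# A two-link constraint chain runs from head through ponytail_01_l to ponytail_02_l, with limited swing, tight twist and wider swing at the lower joint.
# The lower link uses three capsules for volume.
# Capsules on spine_05 and clavicle_l add unconstrained torso and shoulder collision shapes, while head has only a tiny anchor shape.
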PhysicsAsset /BRCosmetics/Characters/Player/Female/Medium/Bodies/F_MED_Binary/Meshes/Parts/F_MED_Binary_Ponytail_l_Physics
bodies {
 clavicle_l [{
   shapeType=Capsule
   length=5.7724
   radius=5.3158
   shapeName=clavicle_l_capsule
   center=<9.2661 -0.0838 0.1635>
   rotation=<rot -83.6899 0.0023 -8.5678>
  }]
 head [{shapeType=Capsule, length=0, radius=0.5, shapeName=head_capsule}]
 ponytail_01_l [{shapeType=Capsule, length=5.6499, radius=1.5988, shapeName=ponytail_01_l_capsule, center=<4.3012 0 0>, rotation=<rot -90 179.9999 179.9997>}]
 ponytail_02_l [
  {shapeType=Capsule, length=8.3218, radius=1.86, shapeName=ponytail_02_l_capsule, center=<6.0585 0 0>, rotation=<rot -90 30.6199 -37.8746>}
  {shapeType=Capsule, length=5.3221, radius=1.5111, shapeName=ponytail_02_l_capsule, center=<6.0531 0.3818 -1.747>, rotation=<rot 90 0 7.5157>}
  {shapeType=Capsule, length=5.3319, radius=1.5933, shapeName=ponytail_02_l_capsule, center=<6.1186 0.3847 1.9016>, rotation=<rot 90 0 8.2016>}
 ]
 spine_05 [
  {shapeType=Capsule, length=13.5525, radius=6.675, shapeName=spine_05_capsule, center=<4.4526 -0.8868 0>}
  {
   shapeType=Capsule
   length=12.8457
   radius=6.9118
   shapeName=spine_05_capsule
   center=<12.7675 -9.5401 -9.6333>
   rotation=<rot 0 -0.0001 -9.1016>
  }
 ]
}
constraints [
 {bone1Name=ponytail_02_l, bone2Name=ponytail_01_l, swing1LimitDegrees=40, swing2LimitDegrees=25, twistLimitDegrees=5}
 {bone1Name=ponytail_01_l, bone2Name=head, swing1LimitDegrees=30, swing2LimitDegrees=15, twistLimitDegrees=5}
]
