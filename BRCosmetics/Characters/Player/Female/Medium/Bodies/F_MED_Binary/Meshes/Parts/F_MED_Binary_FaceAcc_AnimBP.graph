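# Post-processing style AnimBlueprint for the female Binary character's face accessory part, adding secondary motion to hair, bangs and ponytails.
# It copies the attached parent's pose, runs a Control Rig, then chains two RigidBody simulations and per-bone spring-driven AnimDynamics.
# Each update sets the RigidBody alphas to full while skydiving, zero during melee, otherwise one minus upper-arm animation curves.
# Customization load caches the body mesh, which nothing here reads.
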AnimBlueprint /BRCosmetics/Characters/Player/Female/Medium/Bodies/F_MED_Binary/Meshes/Parts/F_MED_Binary_FaceAcc_AnimBP
target_skeleton=&/F_MED_Binary_FaceAcc_Skeleton
graphs [
 AnimGraph {nodes [
   n1 = Control Rig(Source=Copy Pose From Mesh() {bUseAttachedParent=true}) {
    alphaBoolBlend {blendOption=Linear}
    controlRigAssetReference {blueprintRigClass=&/F_MED_Binary_FaceAcc_CtrlRig.F_MED_Binary_FaceAcc_CtrlRig_C}
   }
   n2 = RigidBody(ComponentPose=Local To Component(LocalPose=@n1), Alpha=Get Upperarm_r_Physics_Alpha() -> (Upperarm_r_Physics_Alpha:double)) {
    bClampLinearTranslationLimitToRefPose=true
    bOverrideWorldGravity=true
    overridePhysicsAsset=&/F_MED_Binary_FaceAcc_Physics
    simulationSpace=BaseBoneSpace
    alphaBoolBlend {blendOption=Linear}
    alphaScaleBiasClamp {bInterpResult=true, interpSpeedIncreasing=2.5}
    baseBoneRef {boneName=spine_04}
    componentAppliedLinearAccClamp=<3000 3000 3000>
    componentLinearAccScale=<0.4 0.3 0.3>
    componentLinearVelScale=<0.4 0.3 0.3>
    overrideWorldGravity {z=-980}
   }
   n3 = RigidBody(ComponentPose=@n2, Alpha=Get Upperarm_l_Physics_Alpha() -> (Upperarm_l_Physics_Alpha:double)) {
    bClampLinearTranslationLimitToRefPose=true
    bOverrideWorldGravity=true
    overridePhysicsAsset=&/F_MED_Binary_Ponytail_l_Physics
    simulationSpace=BaseBoneSpace
    alphaBoolBlend {blendOption=Linear}
    alphaScaleBiasClamp {bInterpResult=true, interpSpeedIncreasing=2.5}
    baseBoneRef {boneName=spine_04}
    componentAppliedLinearAccClamp=<3000 3000 3000>
    componentLinearAccScale=<0.4 0.3 0.3>
    componentLinearVelScale=<0.4 0.3 0.3>
    overrideWorldGravity {z=-980}
   }
   n4 = Anim Dynamics - Bone: dyn_bangs_01_l(ComponentPose=@n3, Alpha=1, ExternalForce="0,0,0") {
    angularDampingOverride=1
    angularSpringConstant=500
    bAngularSpring=true
    bOverrideAngularDamping=true
    bUseGravityOverride=true
    gravityScale=0.35
    simulationSpace=RootRelative
    alphaBoolBlend {blendOption=Linear}
    boundBone {boneName=dyn_bangs_01_l}
    componentAppliedLinearAccClamp=<2000 2000 2000>
    componentLinearAccScale=<0.2 0.2 0.3>
    componentLinearVelScale=<0.2 0.2 0.3>
    gravityOverride {z=-980}
    physicsBodyDefinitions [{
      boundBone {boneName=dyn_bangs_01_l}
      boxExtents=<10 10 10>
      constraintSetup {angularLimitsMax=<0 0 5>, angularLimitsMin=<0 0 0>, angularTarget=<0 0 0>}
      localJointOffset=<1.5 0 0>
     }]
   }
   n5 = Anim Dynamics - Bone: dyn_hair_01_l(ComponentPose=@n4, Alpha=1, ExternalForce="0,0,0") {
    angularDampingOverride=0.98
    angularSpringConstant=500
    bAngularSpring=true
    bOverrideAngularDamping=true
    bUseGravityOverride=true
    gravityScale=0.35
    simulationSpace=RootRelative
    alphaBoolBlend {blendOption=Linear}
    boundBone {boneName=dyn_hair_01_l}
    componentAppliedLinearAccClamp=<2000 2000 2000>
    componentLinearAccScale=<0.2 0.2 0.3>
    componentLinearVelScale=<0.2 0.2 0.3>
    gravityOverride {z=-980}
    physicsBodyDefinitions [{
      boundBone {boneName=dyn_hair_01_l}
      boxExtents=<10 10 10>
      constraintSetup {angularLimitsMax=<0 5 5>, angularLimitsMin=<0 -2 -2>, angularTarget=<0 0 0>}
      localJointOffset=<3 0 0>
     }]
   }
   n6 = Anim Dynamics - Bone: dyn_ponytail_hair_01_l(ComponentPose=@n5, Alpha=1, ExternalForce="0,0,0") {
    angularDampingOverride=0.98
    angularSpringConstant=500
    bAngularSpring=true
    bOverrideAngularDamping=true
    bUseGravityOverride=true
    gravityScale=0.35
    simulationSpace=RootRelative
    alphaBoolBlend {blendOption=Linear}
    boundBone {boneName=dyn_ponytail_hair_01_l}
    componentAppliedLinearAccClamp=<2000 2000 2000>
    componentLinearAccScale=<0.3 0.3 0.3>
    componentLinearVelScale=<1 1 1>
    gravityOverride {z=-980}
    physicsBodyDefinitions [{
      boundBone {boneName=dyn_ponytail_hair_01_l}
      boxExtents=<10 10 10>
      constraintSetup {angularLimitsMax=<0 5 5>, angularLimitsMin=<0 -2.5 -5>, angularTarget=<0 0 0>}
      localJointOffset=<2 0 0>
     }]
   }
   n7 = Anim Dynamics - Bone: dyn_bangs_01_r(ComponentPose=@n6, Alpha=1, ExternalForce="0,0,0") {
    angularDampingOverride=1
    angularSpringConstant=500
    bAngularSpring=true
    bOverrideAngularDamping=true
    bUseGravityOverride=true
    gravityScale=0.35
    simulationSpace=RootRelative
    alphaBoolBlend {blendOption=Linear}
    boundBone {boneName=dyn_bangs_01_r}
    componentAppliedLinearAccClamp=<2000 2000 2000>
    componentLinearAccScale=<0.2 0.2 0.3>
    componentLinearVelScale=<0.2 0.2 0.3>
    gravityOverride {z=-980}
    physicsBodyDefinitions [{
      boundBone {boneName=dyn_bangs_01_r}
      boxExtents=<10 10 10>
      constraintSetup {angularLimitsMax=<0 0 5>, angularLimitsMin=<0 0 0>, angularTarget=<0 0 0>}
      localJointOffset=<1.5 0 0>
     }]
   }
   n8 = Anim Dynamics - Bone: dyn_hair_01_r(ComponentPose=@n7, Alpha=1, ExternalForce="0,0,0") {
    angularDampingOverride=0.98
    angularSpringConstant=500
    bAngularSpring=true
    bOverrideAngularDamping=true
    bUseGravityOverride=true
    gravityScale=0.35
    simulationSpace=RootRelative
    alphaBoolBlend {blendOption=Linear}
    boundBone {boneName=dyn_hair_01_r}
    componentAppliedLinearAccClamp=<2000 2000 2000>
    componentLinearAccScale=<0.5 0.5 0.5>
    componentLinearVelScale=<0.9 0.9 0.9>
    gravityOverride {z=-980}
    physicsBodyDefinitions [{
      boundBone {boneName=dyn_hair_01_r}
      boxExtents=<10 10 10>
      constraintSetup {angularLimitsMax=<0 2.5 2.5>, angularLimitsMin=<0 -2.5 -2.5>, angularTarget=<0 0 0>}
      localJointOffset=<2 0 0>
     }]
   }
   n9 = Anim Dynamics - Bone: dyn_hair_02_r(ComponentPose=@n8, Alpha=1, ExternalForce="0,0,0") {
    angularDampingOverride=0.98
    angularSpringConstant=500
    bAngularSpring=true
    bOverrideAngularDamping=true
    bUseGravityOverride=true
    gravityScale=0.35
    simulationSpace=RootRelative
    alphaBoolBlend {blendOption=Linear}
    boundBone {boneName=dyn_hair_02_r}
    componentAppliedLinearAccClamp=<2000 2000 2000>
    componentLinearAccScale=<0.5 0.5 0.5>
    componentLinearVelScale=<0.9 0.9 0.9>
    gravityOverride {z=-980}
    physicsBodyDefinitions [{
      boundBone {boneName=dyn_hair_02_r}
      boxExtents=<10 10 10>
      constraintSetup {angularLimitsMax=<0 2.5 2.5>, angularLimitsMin=<0 -2.5 -2.5>, angularTarget=<0 0 0>}
      localJointOffset=<2 0 0>
     }]
   }
   n10 = Anim Dynamics - Bone: dyn_ponytail_hair_01_r(ComponentPose=@n9, Alpha=1, ExternalForce="0,0,0") {
    angularDampingOverride=0.98
    angularSpringConstant=1500
    bAngularSpring=true
    bOverrideAngularDamping=true
    bUseGravityOverride=true
    gravityScale=0.35
    simulationSpace=RootRelative
    alphaBoolBlend {blendOption=Linear}
    boundBone {boneName=dyn_ponytail_hair_01_r}
    componentAppliedLinearAccClamp=<2000 2000 2000>
    componentLinearAccScale=<0.2 0.2 0.3>
    componentLinearVelScale=<0.2 0.2 0.3>
    gravityOverride {z=-980}
    physicsBodyDefinitions [{
      boundBone {boneName=dyn_ponytail_hair_01_r}
      boxExtents=<5 5 5>
      constraintSetup {angularLimitsMax=<0 2.5 5>, angularLimitsMin=<0 -2.5 0>, angularTarget=<0 0 0>}
      localJointOffset=<2 0 0>
     }]
   }
   n11 = Anim Dynamics - Bone: dyn_ponytail_hair_02_r(ComponentPose=@n10, Alpha=1, ExternalForce="0,0,0") {
    angularDampingOverride=0.98
    angularSpringConstant=1500
    bAngularSpring=true
    bOverrideAngularDamping=true
    bUseGravityOverride=true
    gravityScale=0.35
    simulationSpace=RootRelative
    alphaBoolBlend {blendOption=Linear}
    boundBone {boneName=dyn_ponytail_hair_02_r}
    componentAppliedLinearAccClamp=<2000 2000 2000>
    componentLinearAccScale=<0.2 0.2 0.3>
    componentLinearVelScale=<0.2 0.2 0.3>
    gravityOverride {z=-500}
    physicsBodyDefinitions [{
      boundBone {boneName=dyn_ponytail_hair_02_r}
      boxExtents=<5 5 5>
      constraintSetup {angularLimitsMax=<0 2.5 2.5>, angularLimitsMin=<0 -2.5 -2.5>, angularTarget=<0 0 0>}
      localJointOffset=<2 0 0>
     }]
   }
   Output Pose(Result=Component To Local(ComponentPose=@n11)) {layerGroup=None}
  ]}
 EventGraph {nodes [
   Event BlueprintUpdateAnimation()
   > n1 = Branch(Condition=Get bIsSkyDiving())
   > Set Upperarm_r_Physics_Alpha(Upperarm_r_Physics_Alpha:double=1) -> (Output_Get:double)
   > Set Upperarm_l_Physics_Alpha(Upperarm_l_Physics_Alpha:double=1) -> (Output_Get:double)
   n2 = Branch(execute=@n1.else, Condition=Get bIsPlayingMeleeAnim())
   > Set Upperarm_l_Physics_Alpha(Upperarm_l_Physics_Alpha=0)
   > Set Upperarm_r_Physics_Alpha(Upperarm_r_Physics_Alpha=0)
   Set Upperarm_r_Physics_Alpha(execute=@n2.else, Upperarm_r_Physics_Alpha=float - float(A=1, B=GetCurveValue(CurveName=SPR_Upperarm_r)))
   n3 = GetCurveValue(CurveName=SPR_Upperarm_l)
   n4 = GetCurveValue(CurveName=SPR_Rifle_Upperarm_l)
   > Set Upperarm_l_Physics_Alpha(Upperarm_l_Physics_Alpha=float - float(A=1, B="Max (Float)"(A=@n3, B=@n4)))
   Event OnCharacterCustomizationLoaded()
   > n5 = Cast To PlayerPawn_Athena(Object=TryGetPawnOwner())
   > Set Body_Mesh(Body_Mesh=GetSkeletalMeshForPartType(self=@n5))
  ]}
]
variables [
 Body_Mesh {type=SkeletalMeshComponent}
 Skel_Mesh {type=SkeletalMeshComponent}
 Upperarm_l_Physics_Alpha {type=number}
 Upperarm_r_Physics_Alpha {type=number}
]
